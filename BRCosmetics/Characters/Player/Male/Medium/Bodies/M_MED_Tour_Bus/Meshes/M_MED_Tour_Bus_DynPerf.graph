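# Physics asset for the Tour Bus male medium body cosmetic, simulating secondary motion of its hoodie, back flaps and strings.
# Each dyn_ capsule chain uses tight swing and twist limits and hangs from a tiny MSTR anchor body.
# Hoodie anchors attach to spine_01 and flap and string anchors to the pelvis, with free angular motion.
# Larger capsules on pelvis, spine_01 and both thighs provide body collision volumes.
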PhysicsAsset /BRCosmetics/Characters/Player/Male/Medium/Bodies/M_MED_Tour_Bus/Meshes/M_MED_Tour_Bus_DynPerf
bodies {
 dyn_flap_bk_01 [{shapeType=Capsule, length=13, radius=1, shapeName=dyn_flap_bk_01_capsule, center=<3.5831 0 0.7>, rotation=<rot 0 0 90>}]
 dyn_flap_bk_MSTR [{shapeType=Capsule, length=0.2, radius=0.2}]
 dyn_flap_bk_l_01 [{shapeType=Capsule, length=12, radius=1, shapeName=dyn_flap_bk_l_01_capsule, center=<4.822 0 -0.5>, rotation=<rot 0 0 100.8929>}]
 dyn_flap_bk_l_MSTR [{shapeType=Capsule, length=0.2, radius=0.2}]
 dyn_flap_bk_r_01 [{shapeType=Capsule, length=12, radius=1, shapeName=dyn_flap_bk_r_01_capsule, center=<3.5831 0 0>, rotation=<rot 0 0 90>}]
 dyn_flap_bk_r_MSTR [{shapeType=Capsule, length=0.2, radius=0.2}]
 dyn_flap_string_l_01 [{shapeType=Capsule, length=4.5, radius=1, shapeName=dyn_flap_string_l_01_capsule, center=<4.051 0 -0.2528>, rotation=<rot 90 0 0>}]
 dyn_flap_string_l_02 [{shapeType=Capsule, length=7, radius=1, shapeName=dyn_flap_string_l_02_capsule, center=<3.1907 -0.2352 -0.1193>, rotation=<rot 90 0 0>}]
 dyn_flap_string_l_MSTR [{shapeType=Capsule, length=0.2, radius=0.2}]
 dyn_flap_string_r_01 [{shapeType=Capsule, length=5, radius=1, shapeName=dyn_flap_string_r_01_capsule, center=<2.879 0 0.0801>, rotation=<rot 90 0 0>}]
 dyn_flap_string_r_02 [{shapeType=Capsule, length=8, radius=1, shapeName=dyn_flap_string_r_02_capsule, center=<2.8204 0 0.08>, rotation=<rot 90 0 0>}]
 dyn_flap_string_r_MSTR [{shapeType=Capsule, length=0.2, radius=0.2}]
 dyn_hoodie_bk_01 [{shapeType=Capsule, length=8, radius=1, shapeName=dyn_hoodie_bk_01_capsule, center=<3.6661 0 0>, rotation=<rot 90 0 -0.3058>}]
 dyn_hoodie_bk_02 [{shapeType=Capsule, length=13, radius=0.75, shapeName=dyn_hoodie_bk_02_capsule, center=<5.0184 0 0>, rotation=<rot 90 0 -0.3058>}]
 dyn_hoodie_bk_MSTR [{shapeType=Capsule, length=0.2, radius=0.2}]
 dyn_hoodie_bk_l_01 [{shapeType=Capsule, length=7, radius=2, shapeName=dyn_hoodie_bk_l_01_capsule, center=<3.5925 0 0>, rotation=<rot 77.4327 0 -0.3058>}]
 dyn_hoodie_bk_l_02 [{shapeType=Capsule, length=13.25, radius=0.75, shapeName=dyn_hoodie_bk_l_02_capsule, center=<5.0051 0 0>, rotation=<rot 90 -11.4904 -11.7961>}]
 dyn_hoodie_bk_l_MSTR [{shapeType=Capsule, length=0.2, radius=0.2}]
 dyn_hoodie_bk_r_01 [{shapeType=Capsule, length=6.5, radius=2.75, shapeName=dyn_hoodie_bk_r_01_capsule, center=<3.4692 0 0.2>, rotation=<rot 76.3101 0 -0.3057>}]
 dyn_hoodie_bk_r_02 [{shapeType=Capsule, length=13.25, radius=0.75, shapeName=dyn_hoodie_bk_r_02_capsule, center=<4.9546 0 1.5>, rotation=<rot 90 11.4904 11.1846>}]
 dyn_hoodie_bk_r_MSTR [{shapeType=Capsule, length=0.2, radius=0.2}]
 dyn_hoodie_fr_l_01 [{shapeType=Capsule, length=20.5, radius=5.25, shapeName=dyn_hoodie_fr_l_01_capsule, center=<3.5915 0 4>, rotation=<rot -90 0 -89.9996>}]
 dyn_hoodie_fr_l_02 [{shapeType=Capsule, length=20.5, radius=5.25, shapeName=dyn_hoodie_fr_l_02_capsule, center=<3.6175 0 4>, rotation=<rot 90 -9.4623 -99.4643>}]
 dyn_hoodie_fr_l_03 [{shapeType=Capsule, length=20.25, radius=5.5, shapeName=dyn_hoodie_fr_l_03_capsule, center=<3.6585 0 4>, rotation=<rot 90 -9.4623 -99.4643>}]
 dyn_hoodie_fr_l_04 [{shapeType=Capsule, length=20.25, radius=5.5, shapeName=dyn_hoodie_fr_l_04_capsule, center=<3.627 0 4>, rotation=<rot 90 -9.4623 -99.4643>}]
 dyn_hoodie_fr_l_MSTR [{shapeType=Capsule, length=0.2, radius=0.2}]
 dyn_hoodie_fr_r_01 [{shapeType=Capsule, length=19.5, radius=5.75, shapeName=dyn_hoodie_fr_r_01_capsule, center=<8.7785 3 4>, rotation=<rot 90 180 90>}]
 dyn_hoodie_fr_r_02 [{shapeType=Capsule, length=21, radius=5.75, shapeName=dyn_hoodie_fr_r_02_capsule, center=<7.4334 3 4>, rotation=<rot 90 180 90>}]
 dyn_hoodie_fr_r_03 [{shapeType=Capsule, length=21.75, radius=6.5, shapeName=dyn_hoodie_fr_r_03_capsule, center=<7.7951 3 4>, rotation=<rot 90 180 90>}]
 dyn_hoodie_fr_r_04 [{shapeType=Capsule, length=25, radius=5.75, shapeName=dyn_hoodie_fr_r_04_capsule, center=<9.4145 3 4>, rotation=<rot 90 180 90>}]
 dyn_hoodie_fr_r_MSTR [{shapeType=Capsule, length=0.2, radius=0.2}]
 dyn_hoodie_l_01 [{shapeType=Capsule, length=11, radius=1.5, shapeName=dyn_hoodie_l_01_capsule, center=<4.8767 0 2>, rotation=<rot 0 0 90>}]
 dyn_hoodie_l_02 [{shapeType=Capsule, length=11, radius=1.5, shapeName=dyn_hoodie_l_02_capsule, center=<3.5434 0 2.8>, rotation=<rot 0 0 90>}]
 dyn_hoodie_l_MSTR [{shapeType=Capsule, length=0.2, radius=0.2}]
 dyn_hoodie_r_01 [{shapeType=Capsule, length=11, radius=1.5, shapeName=dyn_hoodie_r_01_capsule, center=<3.5434 0 0.0002>, rotation=<rot 0 0 89.6942>}]
 dyn_hoodie_r_02 [{shapeType=Capsule, length=11, radius=1.5, shapeName=dyn_hoodie_r_02_capsule, center=<3.5434 0 0.0002>, rotation=<rot 0 0 89.6942>}]
 dyn_hoodie_r_MSTR [{shapeType=Capsule, length=0.2, radius=0.2}]
 pelvis [
  {shapeType=Capsule, length=9.0294, radius=11.8072, center=<-5.9975 0.0259 0.0001>}
  {shapeType=Capsule, length=11.625, radius=7, center=<-16.3298 2.1484 0>}
  {shapeType=Capsule, length=20.2211, radius=9.7789, shapeName=pelvis_capsule, center=<-12.018 -0.4084 0>}
 ]
 spine_01 [{shapeType=Capsule, length=15.4901, radius=11.5115, center=<-1.1431 1.0181 0>, rotation=<rot 90 0 0>}]
 thigh_l [
  {shapeType=Capsule, length=20.1002, radius=10.1913, center=<-6.6898 0.0622 0.3534>, rotation=<rot 66.4197 91.309 93.1933>}
  {shapeType=Capsule, length=20.6468, radius=9.3426, center=<-23.4116 -0.9066 1.0481>, rotation=<rot 66.4188 91.3086 87.539>}
 ]
 thigh_r [{shapeType=Capsule, length=36.3037, radius=10.0311, center=<17.4113 0.5601 -1.493>, rotation=<rot 90 89.6476 92.5877>}]
}
constraints [
 {bone1Name=dyn_flap_bk_01, bone2Name=dyn_flap_bk_MSTR, swing1LimitDegrees=20, swing2LimitDegrees=10, twistLimitDegrees=1}
 {bone1Name=dyn_flap_bk_l_01, bone2Name=dyn_flap_bk_l_MSTR, swing1LimitDegrees=20, swing2LimitDegrees=10, twistLimitDegrees=1}
 {bone1Name=dyn_flap_bk_r_01, bone2Name=dyn_flap_bk_r_MSTR, swing1LimitDegrees=20, swing2LimitDegrees=10, twistLimitDegrees=1}
 {bone1Name=dyn_flap_string_l_01, bone2Name=dyn_flap_string_l_MSTR, swing1LimitDegrees=15, swing2LimitDegrees=45, twistLimitDegrees=10}
 {bone1Name=dyn_flap_string_l_02, bone2Name=dyn_flap_string_l_01, swing1LimitDegrees=45, swing2LimitDegrees=45, twistLimitDegrees=5}
 {bone1Name=dyn_flap_string_r_01, bone2Name=dyn_flap_string_r_MSTR, swing1LimitDegrees=15, swing2LimitDegrees=45, twistLimitDegrees=10}
 {bone1Name=dyn_flap_string_r_02, bone2Name=dyn_flap_string_r_01, swing1LimitDegrees=45, swing2LimitDegrees=45, twistLimitDegrees=5}
 {bone1Name=dyn_hoodie_fr_l_01, bone2Name=dyn_hoodie_fr_l_MSTR, swing1LimitDegrees=5, swing2LimitDegrees=45, twistLimitDegrees=5}
 {bone1Name=dyn_hoodie_fr_l_02, bone2Name=dyn_hoodie_fr_l_01, swing1LimitDegrees=5, swing2LimitDegrees=45, twistLimitDegrees=5}
 {bone1Name=dyn_hoodie_fr_l_03, bone2Name=dyn_hoodie_fr_l_02, swing1LimitDegrees=5, swing2LimitDegrees=45, twistLimitDegrees=5}
 {bone1Name=dyn_hoodie_fr_l_04, bone2Name=dyn_hoodie_fr_l_03, swing1LimitDegrees=5, swing2LimitDegrees=45, twistLimitDegrees=5}
 {bone1Name=dyn_hoodie_l_01, bone2Name=dyn_hoodie_l_MSTR, swing1LimitDegrees=5, swing2LimitDegrees=30, twistLimitDegrees=5}
 {bone1Name=dyn_hoodie_l_02, bone2Name=dyn_hoodie_l_01, swing1LimitDegrees=5, swing2LimitDegrees=30, twistLimitDegrees=5}
 {bone1Name=dyn_hoodie_bk_l_01, bone2Name=dyn_hoodie_bk_l_MSTR, swing1LimitDegrees=5, swing2LimitDegrees=20, twistLimitDegrees=1}
 {bone1Name=dyn_hoodie_bk_l_02, bone2Name=dyn_hoodie_bk_l_01, swing1LimitDegrees=10, swing2LimitDegrees=20, twistLimitDegrees=1}
 {bone1Name=dyn_hoodie_bk_01, bone2Name=dyn_hoodie_bk_MSTR, swing1LimitDegrees=6, swing2LimitDegrees=20, twistLimitDegrees=1}
 {bone1Name=dyn_hoodie_bk_02, bone2Name=dyn_hoodie_bk_01, swing1LimitDegrees=6, swing2LimitDegrees=20, twistLimitDegrees=1}
 {bone1Name=dyn_hoodie_bk_r_01, bone2Name=dyn_hoodie_bk_r_MSTR, swing1LimitDegrees=5, swing2LimitDegrees=20, twistLimitDegrees=1}
 {bone1Name=dyn_hoodie_bk_r_02, bone2Name=dyn_hoodie_bk_r_01, swing1LimitDegrees=10, swing2LimitDegrees=20, twistLimitDegrees=1}
 {bone1Name=dyn_hoodie_r_01, bone2Name=dyn_hoodie_r_MSTR, swing1LimitDegrees=30, swing2LimitDegrees=30, twistLimitDegrees=5}
 {bone1Name=dyn_hoodie_r_02, bone2Name=dyn_hoodie_r_01, swing1LimitDegrees=30, swing2LimitDegrees=30, twistLimitDegrees=5}
 {bone1Name=dyn_hoodie_fr_r_04, bone2Name=dyn_hoodie_fr_r_03, swing1LimitDegrees=5, swing2LimitDegrees=20, twistLimitDegrees=0}
 {bone1Name=dyn_hoodie_fr_r_01, bone2Name=dyn_hoodie_fr_r_MSTR, swing1LimitDegrees=5, swing2LimitDegrees=60, twistLimitDegrees=0}
 {bone1Name=dyn_hoodie_fr_r_02, bone2Name=dyn_hoodie_fr_r_01, swing1LimitDegrees=5, swing2LimitDegrees=20, twistLimitDegrees=0}
 {bone1Name=dyn_hoodie_fr_r_03, bone2Name=dyn_hoodie_fr_r_02, swing1LimitDegrees=5, swing2LimitDegrees=20, twistLimitDegrees=0}
 {
  bone1Name=dyn_hoodie_fr_l_MSTR
  bone2Name=spine_01
  swing1LimitDegrees=45
  swing1Motion=Free
  swing2LimitDegrees=45
  swing2Motion=Free
  twistLimitDegrees=45
  twistMotion=Free
 }
 {
  bone1Name=dyn_hoodie_l_MSTR
  bone2Name=spine_01
  swing1LimitDegrees=45
  swing1Motion=Free
  swing2LimitDegrees=45
  swing2Motion=Free
  twistLimitDegrees=45
  twistMotion=Free
 }
 {
  bone1Name=dyn_hoodie_bk_l_MSTR
  bone2Name=spine_01
  swing1LimitDegrees=45
  swing1Motion=Free
  swing2LimitDegrees=45
  swing2Motion=Free
  twistLimitDegrees=45
  twistMotion=Free
 }
 {
  bone1Name=dyn_hoodie_bk_MSTR
  bone2Name=spine_01
  swing1LimitDegrees=45
  swing1Motion=Free
  swing2LimitDegrees=45
  swing2Motion=Free
  twistLimitDegrees=45
  twistMotion=Free
 }
 {
  bone1Name=dyn_hoodie_bk_r_MSTR
  bone2Name=spine_01
  swing1LimitDegrees=45
  swing1Motion=Free
  swing2LimitDegrees=45
  swing2Motion=Free
  twistLimitDegrees=45
  twistMotion=Free
 }
 {
  bone1Name=dyn_hoodie_r_MSTR
  bone2Name=spine_01
  swing1LimitDegrees=45
  swing1Motion=Free
  swing2LimitDegrees=45
  swing2Motion=Free
  twistLimitDegrees=45
  twistMotion=Free
 }
 {
  bone1Name=dyn_hoodie_fr_r_MSTR
  bone2Name=spine_01
  swing1LimitDegrees=45
  swing1Motion=Free
  swing2LimitDegrees=45
  swing2Motion=Free
  twistLimitDegrees=45
  twistMotion=Free
 }
 {
  bone1Name=dyn_flap_bk_MSTR
  bone2Name=pelvis
  swing1LimitDegrees=45
  swing1Motion=Free
  swing2LimitDegrees=45
  swing2Motion=Free
  twistLimitDegrees=45
  twistMotion=Free
 }
 {
  bone1Name=dyn_flap_bk_l_MSTR
  bone2Name=pelvis
  swing1LimitDegrees=45
  swing1Motion=Free
  swing2LimitDegrees=45
  swing2Motion=Free
  twistLimitDegrees=45
  twistMotion=Free
 }
 {
  bone1Name=dyn_flap_bk_r_MSTR
  bone2Name=pelvis
  swing1LimitDegrees=45
  swing1Motion=Free
  swing2LimitDegrees=45
  swing2Motion=Free
  twistLimitDegrees=45
  twistMotion=Free
 }
 {
  bone1Name=thigh_l
  bone2Name=pelvis
  swing1LimitDegrees=45
  swing1Motion=Free
  swing2LimitDegrees=45
  swing2Motion=Free
  twistLimitDegrees=45
  twistMotion=Free
 }
 {
  bone1Name=thigh_r
  bone2Name=pelvis
  swing1LimitDegrees=45
  swing1Motion=Free
  swing2LimitDegrees=45
  swing2Motion=Free
  twistLimitDegrees=45
  twistMotion=Free
 }
 {
  bone1Name=spine_01
  bone2Name=pelvis
  swing1LimitDegrees=45
  swing1Motion=Free
  swing2LimitDegrees=45
  swing2Motion=Free
  twistLimitDegrees=45
  twistMotion=Free
 }
 {
  bone1Name=dyn_flap_string_l_MSTR
  bone2Name=pelvis
  swing1LimitDegrees=45
  swing1Motion=Free
  swing2LimitDegrees=45
  swing2Motion=Free
  twistLimitDegrees=45
  twistMotion=Free
 }
 {
  bone1Name=dyn_flap_string_r_MSTR
  bone2Name=pelvis
  swing1LimitDegrees=45
  swing1Motion=Free
  swing2LimitDegrees=45
  swing2Motion=Free
  twistLimitDegrees=45
  twistMotion=Free
 }
]
