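# Physics asset for the dangling strings and left collar flap on the male medium Tour Bus body cosmetic.
# Each string is a three-capsule chain with swing and tight twist limits; the right hangs from its master bone, the left from the collar flap.
# The flap hinges on one swing axis from its own master, with twist locked.
# Both masters join spine_05 by free-rotation constraints, and large spine capsules provide torso volume.
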PhysicsAsset /BRCosmetics/Characters/Player/Male/Medium/Bodies/M_MED_Tour_Bus/Meshes/M_MED_Tour_Bus_Dyn_Strings
bodies {
 dyn_collar_flap_l [{shapeType=Capsule, length=5.25, radius=1, shapeName=dyn_collar_flap_l_capsule, center=<2.7706 0 0>, rotation=<rot 90 0 0>}]
 dyn_collar_flap_l_MSTR [{shapeType=Capsule, length=0.2, radius=0.2}]
 dyn_string_l_01 [{shapeType=Capsule, length=3.7205, radius=0.9062, center=<5.8431 0 0>, rotation=<rot 90 0 0>}]
 dyn_string_l_02 [{shapeType=Capsule, length=8.9375, radius=0.9062, center=<4.5702 0 0>, rotation=<rot 90 0 0>}]
 dyn_string_l_03 [{shapeType=Capsule, length=5.5, radius=1, shapeName=dyn_string_l_03_capsule, center=<3.2259 0 0>, rotation=<rot 90 0 0>}]
 dyn_string_r_01 [{shapeType=Capsule, length=3.2486, radius=0.9062, center=<6.0702 0 0>, rotation=<rot 90 0 0>}]
 dyn_string_r_02 [{shapeType=Capsule, length=8.9375, radius=0.9062, center=<4.5702 0 0>, rotation=<rot 90 0 0>}]
 dyn_string_r_03 [{shapeType=Capsule, length=5.75, radius=1, shapeName=dyn_string_r_03_capsule, center=<3.1676 0 0>, rotation=<rot 90 0 0>}]
 dyn_string_r_MSTR [{shapeType=Capsule, length=0.2, radius=0.2}]
 spine_03 [{shapeType=Capsule, length=6.9762, radius=13.1175, center=<8.3164 0.8604 0>}]
 spine_04 [{shapeType=Capsule, length=10.8969, radius=12.8218, center=<3.7357 1.1298 0>}]
 spine_05 [{shapeType=Capsule, length=15.0625, radius=13.25, center=<-2.9534 -0.6047 0>}]
}
constraints [
 {bone1Name=dyn_string_r_01, bone2Name=dyn_string_r_MSTR, swing1LimitDegrees=45, swing2LimitDegrees=45, twistLimitDegrees=5}
 {bone1Name=dyn_string_r_02, bone2Name=dyn_string_r_01, swing1LimitDegrees=45, swing2LimitDegrees=45, twistLimitDegrees=5}
 {bone1Name=dyn_string_r_03, bone2Name=dyn_string_r_02, swing1LimitDegrees=45, swing2LimitDegrees=45, twistLimitDegrees=5}
 {
  bone1Name=dyn_collar_flap_l
  bone2Name=dyn_collar_flap_l_MSTR
  swing1LimitDegrees=20
  swing2LimitDegrees=3
  swing2Motion=Locked
  twistLimitDegrees=5
  twistMotion=Locked
 }
 {bone1Name=dyn_string_l_01, bone2Name=dyn_collar_flap_l, swing1LimitDegrees=45, swing2LimitDegrees=45, twistLimitDegrees=5}
 {bone1Name=dyn_string_l_02, bone2Name=dyn_string_l_01, swing1LimitDegrees=45, swing2LimitDegrees=45, twistLimitDegrees=5}
 {bone1Name=dyn_string_l_03, bone2Name=dyn_string_l_02, swing1LimitDegrees=45, swing2LimitDegrees=45, twistLimitDegrees=5}
 {
  bone1Name=dyn_string_r_MSTR
  bone2Name=spine_05
  swing1LimitDegrees=45
  swing1Motion=Free
  swing2LimitDegrees=45
  swing2Motion=Free
  twistLimitDegrees=45
  twistMotion=Free
 }
 {
  bone1Name=dyn_collar_flap_l_MSTR
  bone2Name=spine_05
  swing1LimitDegrees=45
  swing1Motion=Free
  swing2LimitDegrees=45
  swing2Motion=Free
  twistLimitDegrees=45
  twistMotion=Free
 }
 {
  bone1Name=spine_05
  bone2Name=spine_04
  swing1LimitDegrees=45
  swing1Motion=Free
  swing2LimitDegrees=45
  swing2Motion=Free
  twistLimitDegrees=45
  twistMotion=Free
 }
 {
  bone1Name=spine_04
  bone2Name=spine_03
  swing1LimitDegrees=45
  swing1Motion=Free
  swing2LimitDegrees=45
  swing2Motion=Free
  twistLimitDegrees=45
  twistMotion=Free
 }
]
