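# Animation Blueprint for the M_MED_Tour_Bus body cosmetic, adding secondary motion to its hoodie, collar, strings, flaps and zipper.
# It copies the parent mesh's pose, then selects by bRigidBodyEnabled between a simulated chain and a cheaper constraint-and-corrective fallback.
# The simulated chain combines constraints, additive bone correctives, two RigidBody nodes and zipper AnimDynamics.
# The event graph derives corrective alphas from head/thigh rotations and movement states, and reduces gravity while skydiving without parachute.
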AnimBlueprint /BRCosmetics/Characters/Player/Male/Medium/Bodies/M_MED_Tour_Bus/Meshes/M_MED_Tour_Bus_AnimBP
target_skeleton=&/M_MED_Tour_Bus_Skeleton
graphs [
 AnimGraph {nodes [
   "Transform (Modify) Bone - Bone: dyn_hoodie_fr_r_MSTR"(Alpha=1, Translation="-5,3,0", Rotation="55,0,0", Scale="1,1,1") {
    rotationMode=BMM_Additive
    rotationSpace=BCS_BoneSpace
    translationMode=BMM_Additive
    translationSpace=BCS_BoneSpace
    alphaBoolBlend {blendOption=Linear}
    alphaScaleBiasClamp {bInterpResult=true, interpSpeedDecreasing=100, interpSpeedIncreasing=100}
    boneToModify {boneName=dyn_hoodie_fr_r_MSTR}
    rotation {pitch=55}
    translation=<-5 3>
   }
   "Transform (Modify) Bone - Bone: dyn_hoodie_fr_r_MSTR"(Alpha=Get sprinting alpha() -> (sprinting alpha:double), Translation="-1,3,-0.6", Rotation="-10,17,15", Scale="1,1,1") {
    rotationMode=BMM_Additive
    rotationSpace=BCS_BoneSpace
    translationMode=BMM_Additive
    translationSpace=BCS_BoneSpace
    alphaBoolBlend {blendOption=Linear}
    alphaScaleBiasClamp {bInterpResult=true, interpSpeedIncreasing=100}
    boneToModify {boneName=dyn_hoodie_fr_r_MSTR}
    rotation=<rot -10 17 15>
    translation=<-1 3 -0.6>
   }
   n1 = "Constraint - dyn_hoodie_fr_r_01 (PP)"!(
    ComponentPose=Local To Component(LocalPose=Use cached pose 'Start_cache'())
    Alpha=1
    ConstraintWeights_0=0.9397
    ConstraintWeights_1:float=0.0603
   ) -> (Pose) {
    alphaBoolBlend {blendOption=Linear}
    boneToModify {boneName=dyn_hoodie_fr_r_01}
    constraintSetup [{targetBone {boneName=pelvis}}, {targetBone {boneName=spine_01}}]
    constraintWeights [1,1]
   }
   n2 = "Constraint - dyn_hoodie_fr_l_01 (PP)"!(ComponentPose=@n1, Alpha=1, ConstraintWeights_0=0.9397, ConstraintWeights_1:float=0.0603) -> (Pose) {
    alphaBoolBlend {blendOption=Linear}
    boneToModify {boneName=dyn_hoodie_fr_l_01}
    constraintSetup [{targetBone {boneName=pelvis}}, {targetBone {boneName=spine_01}}]
    constraintWeights [1,1]
   }
   n3 = "Transform (Modify) Bone - Bone: dyn_hoodie_fr_r_MSTR"(ComponentPose=@n2, Alpha=Get thigh_r_rot() -> (thigh_r_rot:double), Translation="0,0,0", Rotation="80,-45,0", Scale="1,1,1") {
    rotationMode=BMM_Additive
    rotationSpace=BCS_BoneSpace
    alphaBoolBlend {blendOption=Linear}
    boneToModify {boneName=dyn_hoodie_fr_r_MSTR}
    rotation {pitch=80, yaw=-45}
   }
   n4 = "Transform (Modify) Bone - Bone: dyn_hoodie_fr_l_01"(ComponentPose=@n3, Alpha=Get thigh_l_rot() -> (thigh_l_rot:double), Translation="2,0,-2", Rotation="79.510,-110.224,-112.171", Scale="1,1,1") {
    rotationMode=BMM_Additive
    rotationSpace=BCS_BoneSpace
    translationMode=BMM_Additive
    translationSpace=BCS_BoneSpace
    alphaBoolBlend {blendOption=Linear}
    boneToModify {boneName=dyn_hoodie_fr_l_01}
    rotation=<rot 79.51 -110.2242 -112.1712>
    translation {x=2, z=-2}
   }
   n5 = "Constraint - dyn_flap_string_r_MSTR (PP)"!(ComponentPose=@n4, Alpha=1, ConstraintWeights_0=0.7546, ConstraintWeights_1:float=0.2454) -> (Pose) {
    alphaBoolBlend {blendOption=Linear}
    boneToModify {boneName=dyn_flap_string_r_MSTR}
    constraintSetup [{targetBone {boneName=thigh_twist_01_r}}, {targetBone {boneName=pelvis}}]
    constraintWeights [1,1]
   }
   n6 = "Constraint - dyn_flap_string_l_MSTR (PP)"!(ComponentPose=@n5, Alpha=1, ConstraintWeights_0=0.6565, ConstraintWeights_1:float=0.3435) -> (Pose) {
    alphaBoolBlend {blendOption=Linear}
    boneToModify {boneName=dyn_flap_string_l_MSTR}
    constraintSetup [{targetBone {boneName=thigh_twist_01_l}}, {targetBone {boneName=pelvis}}]
    constraintWeights [1,1]
   }
   n7 = "Transform (Modify) Bone - Bone: dyn_flap_string_r_MSTR"(ComponentPose=@n6, Alpha=Get thigh_r_rot(), Translation="0,0,0", Rotation="45,0,0", Scale="1,1,1") {
    rotationMode=BMM_Additive
    rotationSpace=BCS_BoneSpace
    translationMode=BMM_Additive
    translationSpace=BCS_BoneSpace
    alphaBoolBlend {blendOption=Linear}
    boneToModify {boneName=dyn_flap_string_r_MSTR}
    rotation {pitch=45}
   }
   n8 = "Transform (Modify) Bone - Bone: dyn_flap_string_l_MSTR"(ComponentPose=@n7, Alpha=Get thigh_l_rot(), Translation="0,0,0", Rotation="45,0,0", Scale="1,1,1") {
    rotationMode=BMM_Additive
    rotationSpace=BCS_BoneSpace
    translationMode=BMM_Additive
    translationSpace=BCS_BoneSpace
    alphaBoolBlend {blendOption=Linear}
    boneToModify {boneName=dyn_flap_string_l_MSTR}
    rotation {pitch=45}
   }
   n9 = "Constraint - dyn_hoodie_bk_l_01 (P)"(ComponentPose=@n8, Alpha=1, ConstraintWeights_0=1) {
    alphaBoolBlend {blendOption=Linear}
    boneToModify {boneName=dyn_hoodie_bk_l_01}
    constraintSetup [{targetBone {boneName=pelvis}}]
    constraintWeights [1]
   }
   n10 = "Constraint - dyn_hoodie_bk_l_02 (P)"(ComponentPose=@n9, Alpha=1, ConstraintWeights_0=1) {
    alphaBoolBlend {blendOption=Linear}
    boneToModify {boneName=dyn_hoodie_bk_l_02}
    constraintSetup [{targetBone {boneName=pelvis}}]
    constraintWeights [1]
   }
   n11 = "Constraint - dyn_hoodie_bk_01 (P)"(ComponentPose=@n10, Alpha=1, ConstraintWeights_0=1) {
    alphaBoolBlend {blendOption=Linear}
    boneToModify {boneName=dyn_hoodie_bk_01}
    constraintSetup [{targetBone {boneName=pelvis}}]
    constraintWeights [1]
   }
   n12 = "Constraint - dyn_hoodie_bk_02 (P)"(ComponentPose=@n11, Alpha=1, ConstraintWeights_0=1) {
    alphaBoolBlend {blendOption=Linear}
    boneToModify {boneName=dyn_hoodie_bk_02}
    constraintSetup [{targetBone {boneName=pelvis}}]
    constraintWeights [1]
   }
   n13 = "Constraint - dyn_hoodie_bk_r_01 (P)"(ComponentPose=@n12, Alpha=1, ConstraintWeights_0=1) {
    alphaBoolBlend {blendOption=Linear}
    boneToModify {boneName=dyn_hoodie_bk_r_01}
    constraintSetup [{targetBone {boneName=pelvis}}]
    constraintWeights [1]
   }
   n14 = "Constraint - dyn_hoodie_bk_r_02 (P)"(ComponentPose=@n13, Alpha=1, ConstraintWeights_0=1) {
    alphaBoolBlend {blendOption=Linear}
    boneToModify {boneName=dyn_hoodie_bk_r_02}
    constraintSetup [{targetBone {boneName=pelvis}}]
    constraintWeights [1]
   }
   n15 = "Constraint - dyn_flap_bk_l_01 (RR)"!(ComponentPose=@n14, Alpha=1, ConstraintWeights_0=0.5, ConstraintWeights_1:float=0.5) -> (Pose) {
    alphaBoolBlend {blendOption=Linear}
    boneToModify {boneName=dyn_flap_bk_l_01}
    constraintSetup [{transformType=Rotation, targetBone {boneName=pelvis}}, {transformType=Rotation, targetBone {boneName=thigh_l}}]
    constraintWeights [1,1]
   }
   n16 = Blend Poses by bool(BlendPose_0=Use cached pose 'RB_cache'(), BlendPose_1=Component To Local(ComponentPose=@n15), bActiveValue=Get bRigidBodyEnabled()) {blendType=Linear}
   Output Pose(Result=Component To Local(ComponentPose=Local To Component(LocalPose=@n16))) {layerGroup=None}
   n17 = Get rigid body enabled alpha() -> (rigid body enabled alpha:double)
   n18 = "Constraint - dyn_string_r_01 (R)"(ComponentPose=Local To Component(LocalPose=Use cached pose 'Start_cache'()), Alpha=@n17, ConstraintWeights_0=1) {
    alphaBoolBlend {blendOption=Linear}
    boneToModify {boneName=dyn_string_r_01}
    constraintSetup [{transformType=Rotation, targetBone {boneName=spine_05}}]
    constraintWeights [1]
   }
   n19 = "Constraint - dyn_string_r_02 (R)"(ComponentPose=@n18, Alpha=@n17, ConstraintWeights_0=1) {
    alphaBoolBlend {blendOption=Linear}
    boneToModify {boneName=dyn_string_r_02}
    constraintSetup [{transformType=Rotation, targetBone {boneName=spine_04}}]
    constraintWeights [1]
   }
   n20 = "Constraint - dyn_string_r_03 (R)"(ComponentPose=@n19, Alpha=@n17, ConstraintWeights_0=1) {
    alphaBoolBlend {blendOption=Linear}
    boneToModify {boneName=dyn_string_r_03}
    constraintSetup [{transformType=Rotation, targetBone {boneName=spine_03}}]
    constraintWeights [1]
   }
   n21 = "Constraint - dyn_string_l_01 (R)"(ComponentPose=@n20, Alpha=@n17, ConstraintWeights_0=1) {
    alphaBoolBlend {blendOption=Linear}
    boneToModify {boneName=dyn_string_l_01}
    constraintSetup [{transformType=Rotation, targetBone {boneName=spine_05}}]
    constraintWeights [1]
   }
   n22 = "Constraint - dyn_string_l_02 (R)"(ComponentPose=@n21, Alpha=@n17, ConstraintWeights_0=1) {
    alphaBoolBlend {blendOption=Linear}
    boneToModify {boneName=dyn_string_l_02}
    constraintSetup [{transformType=Rotation, targetBone {boneName=spine_04}}]
    constraintWeights [1]
   }
   n23 = "Constraint - dyn_string_l_03 (R)"(ComponentPose=@n22, Alpha=@n17, ConstraintWeights_0=1) {
    alphaBoolBlend {blendOption=Linear}
    boneToModify {boneName=dyn_string_l_03}
    constraintSetup [{transformType=Rotation, targetBone {boneName=spine_03}}]
    constraintWeights [1]
   }
   n24 = "Constraint - dyn_hoodie_fr_r_01 (R)"(ComponentPose=@n23, Alpha=@n17, ConstraintWeights_0=1) {
    alphaBoolBlend {blendOption=Linear}
    boneToModify {boneName=dyn_hoodie_fr_r_01}
    constraintSetup [{transformType=Rotation, targetBone {boneName=spine_02}}]
    constraintWeights [1]
   }
   n25 = "Constraint - dyn_hoodie_fr_l_01 (R)"(ComponentPose=@n24, Alpha=@n17, ConstraintWeights_0=1) {
    alphaBoolBlend {blendOption=Linear}
    boneToModify {boneName=dyn_hoodie_fr_l_01}
    constraintSetup [{transformType=Rotation, targetBone {boneName=spine_02}}]
    constraintWeights [1]
   }
   n26 = "Constraint - dyn_hoodie_fr_r_03 (R)"(ComponentPose=@n25, Alpha=@n17, ConstraintWeights_0=1) {
    alphaBoolBlend {blendOption=Linear}
    boneToModify {boneName=dyn_hoodie_fr_r_03}
    constraintSetup [{transformType=Rotation, targetBone {boneName=thigh_r}}]
    constraintWeights [1]
   }
   n27 = "Constraint - dyn_hoodie_fr_l_03 (R)"(ComponentPose=@n26, Alpha=@n17, ConstraintWeights_0=1) {
    alphaBoolBlend {blendOption=Linear}
    boneToModify {boneName=dyn_hoodie_fr_l_03}
    constraintSetup [{transformType=Rotation, targetBone {boneName=thigh_l}}]
    constraintWeights [1]
   }
   n28 = "Constraint - dyn_flap_bk_l_01 (R)"(ComponentPose=@n27, Alpha=@n17, ConstraintWeights_0=1) {
    alphaBoolBlend {blendOption=Linear}
    boneToModify {boneName=dyn_flap_bk_l_01}
    constraintSetup [{transformType=Rotation, targetBone {boneName=thigh_r}}]
    constraintWeights [1]
   }
   n29 = "Constraint - dyn_flap_bk_01 (RR)"!(ComponentPose=@n28, Alpha=@n17, ConstraintWeights_0=0.5, ConstraintWeights_1:float=0.5) -> (Pose) {
    alphaBoolBlend {blendOption=Linear}
    boneToModify {boneName=dyn_flap_bk_01}
    constraintSetup [{transformType=Rotation, targetBone {boneName=thigh_r}}, {transformType=Rotation, targetBone {boneName=thigh_l}}]
    constraintWeights [1,1]
   }
   n30 = "Constraint - dyn_flap_bk_r_01 (R)"(ComponentPose=@n29, Alpha=@n17, ConstraintWeights_0=1) {
    alphaBoolBlend {blendOption=Linear}
    boneToModify {boneName=dyn_flap_bk_r_01}
    constraintSetup [{transformType=Rotation, targetBone {boneName=thigh_r}}]
    constraintWeights [1]
   }
   n31 = "Transform (Modify) Bone - Bone: dyn_collar_fr_l"(ComponentPose=@n30, Alpha=Get head l() -> (head l:double), Translation="0,0,0", Rotation="50,1.822,20", Scale="1,1,1") {
    rotationMode=BMM_Additive
    rotationSpace=BCS_BoneSpace
    translationMode=BMM_Additive
    translationSpace=BCS_BoneSpace
    alphaBoolBlend {blendOption=Linear}
    alphaScaleBiasClamp {bInterpResult=true}
    boneToModify {boneName=dyn_collar_fr_l}
    rotation=<rot 50 1.822 20>
   }
   n32 = "Transform (Modify) Bone - Bone: dyn_collar_fr_r"(ComponentPose=@n31, Alpha=Get head r() -> (head r:double), Translation="0,0,0", Rotation="24.899,-1.432,-10.509", Scale="1,1,1") {
    rotationMode=BMM_Additive
    rotationSpace=BCS_BoneSpace
    translationMode=BMM_Additive
    translationSpace=BCS_BoneSpace
    alphaBoolBlend {blendOption=Linear}
    alphaScaleBiasClamp {bInterpResult=true}
    boneToModify {boneName=dyn_collar_fr_r}
    rotation=<rot 24.899 -1.432 -10.5092>
   }
   n33 = "Constraint - dyn_flap_string_l_MSTR (PPP)"!(ComponentPose=@n32, Alpha=1, ConstraintWeights_0=0.2432, ConstraintWeights_1:float=0.7478, ConstraintWeights_2:float=0.009) -> (Pose) {
    alphaBoolBlend {blendOption=Linear}
    boneToModify {boneName=dyn_flap_string_l_MSTR}
    constraintSetup [{targetBone {boneName=pelvis}}, {targetBone {boneName=thigh_twist_01_l}}, {targetBone {boneName=thigh_l}}]
    constraintWeights [1,1,1]
   }
   n34 = "Constraint - dyn_flap_string_r_MSTR (PPP)"!(ComponentPose=@n33, Alpha=1, ConstraintWeights_0=0.009, ConstraintWeights_1:float=0.2432, ConstraintWeights_2:float=0.7478) -> (Pose) {
    alphaBoolBlend {blendOption=Linear}
    boneToModify {boneName=dyn_flap_string_r_MSTR}
    constraintSetup [{targetBone {boneName=thigh_r}}, {targetBone {boneName=pelvis}}, {targetBone {boneName=thigh_twist_01_r}}]
    constraintWeights [1,1,1]
   }
   n35 = "Transform (Modify) Bone - Bone: dyn_hoodie_bk_l_MSTR"(ComponentPose=@n34, Alpha=1, Translation="-0.000,0.000,5", Rotation="0,0,0", Scale="1,1,1") {
    rotationMode=BMM_Additive
    rotationSpace=BCS_BoneSpace
    translationMode=BMM_Additive
    translationSpace=BCS_BoneSpace
    alphaBoolBlend {blendOption=Linear}
    alphaScaleBiasClamp {interpSpeedDecreasing=30, interpSpeedIncreasing=30}
    boneToModify {boneName=dyn_hoodie_bk_l_MSTR}
    translation {z=5}
   }
   n36 = "Transform (Modify) Bone - Bone: dyn_hoodie_bk_MSTR"(ComponentPose=@n35, Alpha=1, Translation="-0.000,0.000,3", Rotation="0,0,0", Scale="1,1,1") {
    rotationMode=BMM_Additive
    rotationSpace=BCS_BoneSpace
    translationMode=BMM_Additive
    translationSpace=BCS_BoneSpace
    alphaBoolBlend {blendOption=Linear}
    alphaScaleBiasClamp {interpSpeedDecreasing=30, interpSpeedIncreasing=30}
    boneToModify {boneName=dyn_hoodie_bk_MSTR}
    translation {z=3}
   }
   n37 = "Transform (Modify) Bone - Bone: dyn_hoodie_bk_r_MSTR"(ComponentPose=@n36, Alpha=1, Translation="0.000,0.000,5", Rotation="0,0,0", Scale="1,1,1") {
    rotationMode=BMM_Additive
    rotationSpace=BCS_BoneSpace
    translationMode=BMM_Additive
    translationSpace=BCS_BoneSpace
    alphaBoolBlend {blendOption=Linear}
    alphaScaleBiasClamp {interpSpeedDecreasing=30, interpSpeedIncreasing=30}
    boneToModify {boneName=dyn_hoodie_bk_r_MSTR}
    translation {z=5}
   }
   n38 = "Transform (Modify) Bone - Bone: dyn_hoodie_bk_l_02"(ComponentPose=@n37, Alpha=1, Translation="-0.000,0.000,3", Rotation="0,0,0", Scale="1,1,1") {
    rotationMode=BMM_Additive
    rotationSpace=BCS_BoneSpace
    translationMode=BMM_Additive
    translationSpace=BCS_BoneSpace
    alphaBoolBlend {blendOption=Linear}
    alphaScaleBiasClamp {interpSpeedDecreasing=30, interpSpeedIncreasing=30}
    boneToModify {boneName=dyn_hoodie_bk_l_02}
    translation {z=3}
   }
   n39 = "Transform (Modify) Bone - Bone: dyn_hoodie_bk_02"(ComponentPose=@n38, Alpha=1, Translation="-0.000,0.000,3", Rotation="0,0,0", Scale="1,1,1") {
    rotationMode=BMM_Additive
    rotationSpace=BCS_BoneSpace
    translationMode=BMM_Additive
    translationSpace=BCS_BoneSpace
    alphaBoolBlend {blendOption=Linear}
    alphaScaleBiasClamp {interpSpeedDecreasing=30, interpSpeedIncreasing=30}
    boneToModify {boneName=dyn_hoodie_bk_02}
    translation {z=3}
   }
   n40 = "Transform (Modify) Bone - Bone: dyn_hoodie_bk_r_02"(ComponentPose=@n39, Alpha=1, Translation="0.000,0.000,3", Rotation="0,0,0", Scale="1,1,1") {
    rotationMode=BMM_Additive
    rotationSpace=BCS_BoneSpace
    translationMode=BMM_Additive
    translationSpace=BCS_BoneSpace
    alphaBoolBlend {blendOption=Linear}
    alphaScaleBiasClamp {interpSpeedDecreasing=30, interpSpeedIncreasing=30}
    boneToModify {boneName=dyn_hoodie_bk_r_02}
    translation {z=3}
   }
   n41 = "Transform (Modify) Bone - Bone: dyn_hoodie_fr_l_MSTR"(ComponentPose=@n40, Alpha=Get CrouchThighLRot() -> (CrouchThighLRot:double), Translation="0,0,0", Rotation="100,-90,-100", Scale="1,1,1") {
    rotationMode=BMM_Additive
    rotationSpace=BCS_BoneSpace
    translationMode=BMM_Additive
    translationSpace=BCS_BoneSpace
    alphaBoolBlend {blendOption=Linear}
    alphaScaleBiasClamp {interpSpeedDecreasing=30, interpSpeedIncreasing=30}
    boneToModify {boneName=dyn_hoodie_fr_l_MSTR}
    rotation=<rot 100 -90 -100>
   }
   n42 = Get crouch alpha() -> (crouch alpha:double)
   n43 = "Transform (Modify) Bone - Bone: dyn_hoodie_bk_l_MSTR"(ComponentPose=@n41, Alpha=@n42, Translation="0.000,-0.000,3", Rotation="0,0,0", Scale="1,1,1") {
    rotationMode=BMM_Additive
    rotationSpace=BCS_BoneSpace
    translationMode=BMM_Additive
    translationSpace=BCS_BoneSpace
    alphaBoolBlend {blendOption=Linear}
    alphaScaleBiasClamp {bInterpResult=true, interpSpeedDecreasing=30, interpSpeedIncreasing=30}
    boneToModify {boneName=dyn_hoodie_bk_l_MSTR}
    translation {z=3}
   }
   n44 = "Transform (Modify) Bone - Bone: dyn_hoodie_bk_r_MSTR"(ComponentPose=@n43, Alpha=@n42, Translation="0.000,-0.000,4", Rotation="0,0,0", Scale="1,1,1") {
    rotationMode=BMM_Additive
    rotationSpace=BCS_BoneSpace
    translationMode=BMM_Additive
    translationSpace=BCS_BoneSpace
    alphaBoolBlend {blendOption=Linear}
    alphaScaleBiasClamp {bInterpResult=true, interpSpeedDecreasing=30, interpSpeedIncreasing=30}
    boneToModify {boneName=dyn_hoodie_bk_r_MSTR}
    translation {z=4}
   }
   n45 = "Transform (Modify) Bone - Bone: dyn_hoodie_l_MSTR"(ComponentPose=@n44, Alpha=@n42, Translation="0,0,2", Rotation="0,0,0", Scale="1,1,1") {
    rotationMode=BMM_Additive
    rotationSpace=BCS_BoneSpace
    translationMode=BMM_Additive
    translationSpace=BCS_BoneSpace
    alphaBoolBlend {blendOption=Linear}
    alphaScaleBiasClamp {bInterpResult=true}
    boneToModify {boneName=dyn_hoodie_l_MSTR}
    translation {z=2}
   }
   Save cached pose 'hoodie_cache'(Pose=Component To Local(ComponentPose=@n45)) {cacheName=hoodie_cache}
   n46 = Component To Local(ComponentPose=Local To Component(LocalPose=Use cached pose 'hoodie_cache'()))
   n47 = Component To Local(ComponentPose=Local To Component(LocalPose=Use cached pose 'hoodie_cache'()))
   n48 = Get running_jumping_bool() -> (running_jumping_bool:bool)
   n49 = "Constraint - dyn_collar_flap_l_MSTR (PPPP)"!(
    ComponentPose=Local To Component(LocalPose=Blend Poses by bool(BlendPose_0=@n47, BlendPose_1=@n46, bActiveValue=@n48) {blendType=Linear})
    Alpha=1
    ConstraintWeights_0=0.1058
    ConstraintWeights_1:float=0.386
    ConstraintWeights_2:float=0.1639
    ConstraintWeights_3:float=0.3443
   ) -> (Pose) {
    alphaBoolBlend {blendOption=Linear}
    boneToModify {boneName=dyn_collar_flap_l_MSTR}
    constraintSetup [{targetBone {boneName=spine_03}}, {targetBone {boneName=spine_04}}, {targetBone {boneName=spine_05}}, {targetBone {boneName=clavicle_l}}]
    constraintWeights [1,1,1,1]
   }
   n50 = "Constraint - dyn_string_r_MSTR (PPPP)"!(
    ComponentPose=@n49
    Alpha=1
    ConstraintWeights_0=0.3443
    ConstraintWeights_1:float=0.1058
    ConstraintWeights_2:float=0.386
    ConstraintWeights_3:float=0.1639
   ) -> (Pose) {
    alphaBoolBlend {blendOption=Linear}
    boneToModify {boneName=dyn_string_r_MSTR}
    constraintSetup [{targetBone {boneName=clavicle_r}}, {targetBone {boneName=spine_03}}, {targetBone {boneName=spine_04}}, {targetBone {boneName=spine_05}}]
    constraintWeights [1,1,1,1]
   }
   n51 = "Transform (Modify) Bone - Bone: dyn_hoodie_fr_r_MSTR"(ComponentPose=@n50, Alpha=Get thigh rotation_r() -> (thigh rotation_r:double), Translation="-1,1,0", Rotation="20,0,20", Scale="1,1,1") {
    rotationMode=BMM_Additive
    rotationSpace=BCS_BoneSpace
    translationMode=BMM_Additive
    translationSpace=BCS_BoneSpace
    alphaBoolBlend {blendOption=Linear}
    alphaScaleBiasClamp {bInterpResult=true}
    boneToModify {boneName=dyn_hoodie_fr_r_MSTR}
    rotation {pitch=20, roll=20}
    translation=<-1 1>
   }
   n52 = Get gravity_override() -> (gravity_override:Vector)
   n53 = RigidBody!(ComponentPose=@n51, Alpha, OverrideWorldGravity:Vector=@n52, ExternalForce) -> (Pose) {
    bClampLinearTranslationLimitToRefPose=true
    bOverrideWorldGravity=true
    overridePhysicsAsset=&/M_MED_Tour_Bus_DynPerf
    simulationSpace=BaseBoneSpace
    alphaBoolBlend {blendOption=Linear}
    baseBoneRef {boneName=pelvis}
    componentAppliedLinearAccClamp=<3000 3000 3000>
    componentLinearAccScale=<0.4 0.3 0.3>
    componentLinearVelScale=<0.4 0.3 0.3>
    overrideWorldGravity {z=-980}
   }
   n54 = RigidBody(ComponentPose=@n53, Alpha=Get rigid body enabled inv alpha() -> (rigid body enabled inv alpha:double), OverrideWorldGravity=@n52) {
    bClampLinearTranslationLimitToRefPose=true
    bOverrideWorldGravity=true
    overridePhysicsAsset=&/M_MED_Tour_Bus_Dyn_Strings
    simulationSpace=BaseBoneSpace
    alphaBoolBlend {blendOption=Linear}
    baseBoneRef {boneName=spine_03}
    componentAppliedLinearAccClamp=<3000 3000 3000>
    componentLinearAccScale=<0.4 0.3 0.3>
    componentLinearVelScale=<0.4 0.3 0.3>
    overrideWorldGravity {z=-980}
   }
   n55 = "Constraint - dyn_zipper_l_MSTR (PPPPPPPPP)"!(
    ComponentPose=@n54
    Alpha=1
    ConstraintWeights_0=0.0455
    ConstraintWeights_1:float=0.0693
    ConstraintWeights_2:float=0.2595
    ConstraintWeights_3:float=0.1516
    ConstraintWeights_4:float=0.026
    ConstraintWeights_5:float=0.107
    ConstraintWeights_6:float=0.1533
    ConstraintWeights_7:float=0.0813
    ConstraintWeights_8:float=0.1065
   ) -> (Pose) {
    alphaBoolBlend {blendOption=Linear}
    boneToModify {boneName=dyn_zipper_l_MSTR}
    constraintSetup [
     {targetBone {boneName=dyn_hoodie_l_02}}
     {targetBone {boneName=dyn_hoodie_l_01}}
     {targetBone {boneName=dyn_hoodie_fr_l_02}}
     {targetBone {boneName=dyn_hoodie_fr_l_01}}
     {targetBone {boneName=dyn_hoodie_fr_l_03}}
     {targetBone {boneName=spine_01}}
     {targetBone {boneName=spine_02}}
     {targetBone {boneName=spine_03}}
     {targetBone {boneName=pelvis}}
    ]
    constraintWeights [1,1,1,1,1,1,1,1,1]
   }
   n56 = "Transform (Modify) Bone - Bone: dyn_hoodie_fr_r_MSTR"(ComponentPose=@n55, Alpha=Get Is Falling() -> (Is Falling:double), Translation="0,0,0", Rotation="10,-30,0", Scale="1,1,1") {
    rotationMode=BMM_Additive
    rotationSpace=BCS_BoneSpace
    translationMode=BMM_Additive
    translationSpace=BCS_BoneSpace
    alphaBoolBlend {blendOption=Linear}
    alphaScaleBiasClamp {bInterpResult=true}
    boneToModify {boneName=dyn_hoodie_fr_r_MSTR}
    rotation {pitch=10, yaw=-30}
   }
   n57 = "Transform (Modify) Bone - Bone: dyn_hoodie_fr_l_03"(ComponentPose=@n56, Alpha=Get crouch alpha(), Translation="0,0,0", Rotation="10,0,0", Scale="1,1,1") {
    rotationMode=BMM_Additive
    rotationSpace=BCS_BoneSpace
    translationMode=BMM_Additive
    translationSpace=BCS_BoneSpace
    alphaBoolBlend {blendOption=Linear}
    alphaScaleBiasClamp {bInterpResult=true}
    boneToModify {boneName=dyn_hoodie_fr_l_03}
    rotation {pitch=10}
   }
   n58 = "Transform (Modify) Bone - Bone: dyn_hoodie_fr_r_MSTR"(ComponentPose=@n57, Alpha=Get thigh rotation_r(), Translation="-1,0,0", Rotation="3,-0.000,0.000", Scale="1,1,1") {
    rotationMode=BMM_Additive
    rotationSpace=BCS_BoneSpace
    translationMode=BMM_Additive
    translationSpace=BCS_BoneSpace
    alphaBoolBlend {blendOption=Linear}
    alphaScaleBiasClamp {bInterpResult=true, interpSpeedIncreasing=100}
    boneToModify {boneName=dyn_hoodie_fr_r_MSTR}
    rotation {pitch=3}
    translation {x=-1}
   }
   n59 = Anim Dynamics - Bone: dyn_zipper_l(ComponentPose=@n58, Alpha=1, ExternalForce="0,0,0") {
    angularDampingOverride=1
    angularSpringConstant=1500
    bAngularSpring=true
    bGravityOverrideInSimSpace=true
    bOverrideAngularDamping=true
    gravityScale=0.35
    simulationSpace=RootRelative
    alphaBoolBlend {blendOption=Linear}
    boundBone {boneName=dyn_zipper_l}
    physicsBodyDefinitions [{
      sphereCollisionRadius=0
      boundBone {boneName=dyn_zipper_l}
      boxExtents=<10 20 20>
      constraintSetup {angularLimitsMax=<0 0 40>, angularLimitsMin=<0 -40 -40>, angularTarget=<0 0 0>}
      localJointOffset=<10 0 0>
     }]
   }
   n60 = Get CrouchThighLRot()
   n61 = "Transform (Modify) Bone - Bone: dyn_hoodie_fr_l_MSTR"(ComponentPose=@n59, Alpha=@n60, Translation="-2,-7,0", Rotation="-5,0,0", Scale="1,1,1") {
    rotationMode=BMM_Additive
    rotationSpace=BCS_BoneSpace
    translationMode=BMM_Additive
    translationSpace=BCS_BoneSpace
    alphaBoolBlend {blendOption=Linear}
    alphaScaleBiasClamp {interpSpeedDecreasing=30, interpSpeedIncreasing=30}
    boneToModify {boneName=dyn_hoodie_fr_l_MSTR}
    rotation {pitch=-5}
    translation=<-2 -7>
   }
   n62 = "Transform (Modify) Bone - Bone: dyn_hoodie_fr_l_02"(ComponentPose=@n61, Alpha=@n60, Translation="-2,0,0", Rotation="0,0,0", Scale="1,1,1") {
    rotationMode=BMM_Additive
    rotationSpace=BCS_BoneSpace
    translationMode=BMM_Additive
    translationSpace=BCS_BoneSpace
    alphaBoolBlend {blendOption=Linear}
    alphaScaleBiasClamp {interpSpeedDecreasing=30, interpSpeedIncreasing=30}
    boneToModify {boneName=dyn_hoodie_fr_l_02}
    translation {x=-2}
   }
   n63 = "Transform (Modify) Bone - Bone: dyn_hoodie_fr_l_03"(ComponentPose=@n62, Alpha=@n60, Translation="-2,0,2", Rotation="0,0,0", Scale="1,1,1") {
    rotationMode=BMM_Additive
    rotationSpace=BCS_BoneSpace
    translationMode=BMM_Additive
    translationSpace=BCS_BoneSpace
    alphaBoolBlend {blendOption=Linear}
    alphaScaleBiasClamp {interpSpeedDecreasing=30, interpSpeedIncreasing=30}
    boneToModify {boneName=dyn_hoodie_fr_l_03}
    translation {x=-2, z=2}
   }
   n64 = "Transform (Modify) Bone - Bone: dyn_hoodie_fr_l_03"(ComponentPose=@n63, Alpha=@n60, Translation="-2,0,0", Rotation="0,0,0", Scale="1,1,1") {
    rotationMode=BMM_Additive
    rotationSpace=BCS_BoneSpace
    translationMode=BMM_Additive
    translationSpace=BCS_BoneSpace
    alphaBoolBlend {blendOption=Linear}
    alphaScaleBiasClamp {interpSpeedDecreasing=30, interpSpeedIncreasing=30}
    boneToModify {boneName=dyn_hoodie_fr_l_03}
    translation {x=-2}
   }
   n65 = "Transform (Modify) Bone - Bone: dyn_hoodie_l_02"(ComponentPose=@n64, Alpha=@n60, Translation="0,0,2", Rotation="0,0,0", Scale="1,1,1") {
    rotationMode=BMM_Additive
    rotationSpace=BCS_BoneSpace
    translationMode=BMM_Additive
    translationSpace=BCS_BoneSpace
    alphaBoolBlend {blendOption=Linear}
    alphaScaleBiasClamp {interpSpeedDecreasing=30, interpSpeedIncreasing=30}
    boneToModify {boneName=dyn_hoodie_l_02}
    translation {z=2}
   }
   n66 = Get thigh rotation_r()
   n67 = "Transform (Modify) Bone - Bone: dyn_hoodie_fr_r_02"(ComponentPose=@n65, Alpha=@n66, Translation="-2,0,0", Rotation="0,0,0", Scale="1,1,1") {
    rotationMode=BMM_Additive
    rotationSpace=BCS_BoneSpace
    translationMode=BMM_Additive
    translationSpace=BCS_BoneSpace
    alphaBoolBlend {blendOption=Linear}
    alphaScaleBiasClamp {interpSpeedDecreasing=30, interpSpeedIncreasing=30}
    boneToModify {boneName=dyn_hoodie_fr_r_02}
    translation {x=-2}
   }
   n68 = "Transform (Modify) Bone - Bone: dyn_hoodie_fr_r_03"(ComponentPose=@n67, Alpha=@n66, Translation="-2,0,0", Rotation="0,0,0", Scale="1,1,1") {
    rotationMode=BMM_Additive
    rotationSpace=BCS_BoneSpace
    translationMode=BMM_Additive
    translationSpace=BCS_BoneSpace
    alphaBoolBlend {blendOption=Linear}
    alphaScaleBiasClamp {interpSpeedDecreasing=30, interpSpeedIncreasing=30}
    boneToModify {boneName=dyn_hoodie_fr_r_03}
    translation {x=-2}
   }
   n69 = "Transform (Modify) Bone - Bone: dyn_hoodie_fr_r_04"(ComponentPose=@n68, Alpha=@n66, Translation="-2,0,0", Rotation="0,0,0", Scale="1,1,1") {
    rotationMode=BMM_Additive
    rotationSpace=BCS_BoneSpace
    translationMode=BMM_Additive
    translationSpace=BCS_BoneSpace
    alphaBoolBlend {blendOption=Linear}
    alphaScaleBiasClamp {interpSpeedDecreasing=30, interpSpeedIncreasing=30}
    boneToModify {boneName=dyn_hoodie_fr_r_04}
    translation {x=-2}
   }
   n70 = Get Is Sprinting() -> (Is Sprinting:double)
   n71 = "Transform (Modify) Bone - Bone: dyn_hoodie_bk_l_MSTR"(ComponentPose=@n69, Alpha=@n70, Translation="0,0,2", Rotation="0,0,0", Scale="1,1,1") {
    rotationMode=BMM_Additive
    rotationSpace=BCS_BoneSpace
    translationMode=BMM_Additive
    translationSpace=BCS_BoneSpace
    alphaBoolBlend {blendOption=Linear}
    alphaScaleBiasClamp {bInterpResult=true, interpSpeedDecreasing=1, interpSpeedIncreasing=50}
    boneToModify {boneName=dyn_hoodie_bk_l_MSTR}
    translation {z=2}
   }
   n72 = "Transform (Modify) Bone - Bone: dyn_hoodie_bk_MSTR"(ComponentPose=@n71, Alpha=@n70, Translation="0,0,1", Rotation="0,0,0", Scale="1,1,1") {
    rotationMode=BMM_Additive
    rotationSpace=BCS_BoneSpace
    translationMode=BMM_Additive
    translationSpace=BCS_BoneSpace
    alphaBoolBlend {blendOption=Linear}
    alphaScaleBiasClamp {bInterpResult=true, interpSpeedDecreasing=1, interpSpeedIncreasing=50}
    boneToModify {boneName=dyn_hoodie_bk_MSTR}
    translation {z=1}
   }
   n73 = "Transform (Modify) Bone - Bone: dyn_hoodie_bk_r_MSTR"(ComponentPose=@n72, Alpha=@n70, Translation="0,0,1", Rotation="0,0,0", Scale="1,1,1") {
    rotationMode=BMM_Additive
    rotationSpace=BCS_BoneSpace
    translationMode=BMM_Additive
    translationSpace=BCS_BoneSpace
    alphaBoolBlend {blendOption=Linear}
    alphaScaleBiasClamp {bInterpResult=true, interpSpeedDecreasing=1, interpSpeedIncreasing=50}
    boneToModify {boneName=dyn_hoodie_bk_r_MSTR}
    translation {z=1}
   }
   n74 = Get Right() -> (Right:double)
   n75 = "Transform (Modify) Bone - Bone: dyn_hoodie_bk_r_MSTR"(ComponentPose=@n73, Alpha=@n74, Translation="0,0,4", Rotation="0,0,0", Scale="1,1,1") {
    rotationMode=BMM_Additive
    rotationSpace=BCS_BoneSpace
    translationMode=BMM_Additive
    translationSpace=BCS_BoneSpace
    alphaBoolBlend {blendInTime=0.2, blendOption=Linear, blendOutTime=0.2}
    alphaScaleBiasClamp {bInterpResult=true}
    boneToModify {boneName=dyn_hoodie_bk_r_MSTR}
    translation {z=4}
   }
   n76 = "Transform (Modify) Bone - Bone: dyn_hoodie_bk_MSTR"(ComponentPose=@n75, Alpha=@n74, Translation="0,0,5", Rotation="0,0,0", Scale="1,1,1") {
    rotationMode=BMM_Additive
    rotationSpace=BCS_BoneSpace
    translationMode=BMM_Additive
    translationSpace=BCS_BoneSpace
    alphaBoolBlend {blendInTime=0.2, blendOption=Linear, blendOutTime=0.2}
    alphaScaleBiasClamp {bInterpResult=true}
    boneToModify {boneName=dyn_hoodie_bk_MSTR}
    translation {z=5}
   }
   Save cached pose 'RB_cache'(Pose=Component To Local(ComponentPose=@n76)) {cacheName=RB_cache}
   Save cached pose 'Start_cache'(Pose=Copy Pose From Mesh() {bUseAttachedParent=true}) {cacheName=Start_cache}
  ]}
 EventGraph {nodes [
   Event BlueprintUpdateAnimation()
   > n1 = Sequence!(execute) -> (then_0, then_1, then_2:Exec, then_3:Exec, then_4:Exec, then_5:Exec, then_6:Exec, then_7:Exec, then_8:Exec)
   n2 = Get bIsSprinting()
   > n3 = Branch(Condition=@n2)
   > Set sprinting alpha(sprinting alpha:double=1) -> (Output_Get:double)
   Set sprinting alpha(execute=@n3.else, sprinting alpha=0)
   n4 = GetPartSkeletalMeshComponent()
   n5 = BreakRotator(InRot=BreakTransform(InTransform=GetDeltaTransformFromRefPose(self=@n4, BoneName=head, BaseName=spine_05)).Rotation)
   Set head l(execute=@n1.then_1, head l:double=MapRangeClamped(Value=@n5, InRangeA=11, InRangeB=30)) -> (Output_Get:double)
   > Set head r(head r:double=MapRangeClamped(Value=@n5, InRangeA=-11, InRangeB=-30)) -> (Output_Get:double)
   n6 = Branch(execute=@n1.then_2, Condition=Get bIsJumping())
   Set jumping alpha(execute=[@n6.then, @n7.then], jumping alpha:double=1) -> (Output_Get:double)
   n7 = Branch(execute=@n6.else, Condition=Get bIsFalling())
   Set jumping alpha(execute=@n7.else, jumping alpha=0)
   n8 = Branch(execute=@n1.then_3, Condition=Get bIsSkyDiving())
   > n9 = Branch(Condition=Get bIsParachuteOpened())
   Set gravity_override(execute=[@n9.then, @n8.else], gravity_override="0, 0,-980")
   Set gravity_override(execute=@n9.else, gravity_override="0, 0,-300")
   n10 = Branch(execute=@n1.then_4, Condition=Get bRigidBodyEnabled())
   > Set rigid body enabled alpha(rigid body enabled alpha:double=0) -> (Output_Get:double)
   > Set rigid body enabled inv alpha(rigid body enabled inv alpha:double=1) -> (Output_Get:double)
   Set rigid body enabled alpha(execute=@n10.else, rigid body enabled alpha=1)
   > Set rigid body enabled inv alpha(rigid body enabled inv alpha=0)
   n11 = Branch(execute=@n1.then_5, Condition=AND Boolean(A=@n2, B=Get bIsJumping()))
   > Set running_jumping_bool(running_jumping_bool:bool=true) -> (Output_Get:bool)
   Set running_jumping_bool(execute=@n11.else, running_jumping_bool=false)
   n12 = BreakTransform(InTransform=GetDeltaTransformFromRefPose(self=GetOwnerSkeletalMeshComponent(), BoneName=thigh_r, BaseName=pelvis))
   Set thigh rotation_r(execute=@n1.then_6, thigh rotation_r:double=MapRangeClamped(Value=BreakRotator(InRot=@n12.Rotation).Yaw, InRangeA=0, InRangeB=-50)) -> (Output_Get:double)
   n13 = BreakTransform(InTransform=GetDeltaTransformFromRefPose(self=GetOwnerSkeletalMeshComponent(), BoneName=thigh_l, BaseName=pelvis))
   > Set thigh_rotation_l(thigh_rotation_l:double=MapRangeClamped(Value=BreakRotator(InRot=@n13.Rotation).Yaw, InRangeA=0, InRangeB=-70)) -> (Output_Get:double)
   > n14 = PlayerMovement(AnimInstance=Self-Reference(), XDotProductCoefficient=100, YDotProductCoefficient=100, ZDotProductCoefficient=100)
   n15 = BreakVector(InVec=MakeVector(X=@n14, Y=@n14.YMovement))
   > Set Left(Left:double="To Float (Boolean)"(InBool="float < float"(A=@n15, B=-50))) -> (Output_Get:double)
   n16 = "float > float"(A=@n15, B=50)
   n17 = NOT Boolean(A=Get bIsCrouching())
   n18 = Get bIsOnGround() -> (bIsOnGround:bool)
   > Set Right(Right:double=float * float(A="To Float (Boolean)"(InBool=@n16), B:bool=AND Boolean(A=@n18, B=@n17)) {is=title K2Node_PromotableOperator}) -> (Output_Get:double)
   > Set Forward(Forward:double="To Float (Boolean)"(InBool="float > float"(A=@n15.Y, B=40))) -> (Output_Get:double)
   > Set Backward(Backward:double="To Float (Boolean)"(InBool="float < float"(A=@n15.Y, B=-40))) -> (Output_Get:double)
   n19 = Get SkelMesh()
   > Set thigh_l_rot(thigh_l_rot:double=GetAlphaFromReferenceTransformDelta(SkeletalMeshComponent=@n19, BoneName=pelvis, BaseName=thigh_l, InputRange=(X=0,Y=90), TwistAxis="0,0,1")) -> (Output_Get:double)
   > Set thigh_r_rot(thigh_r_rot:double=GetAlphaFromReferenceTransformDelta(SkeletalMeshComponent=@n19, BoneName=pelvis, BaseName=thigh_r, InputRange=(X=0,Y=90), TwistAxis="0,0,1")) -> (Output_Get:double)
   n20 = Get rigid body enabled inv alpha()
   Set crouchThighRot(
    execute=@n1.then_7
    crouchThighRot:double=float * float!(A=Get crouch alpha(), B:double=Get thigh rotation_r(), C:double=@n20) -> (ReturnValue)
   ) -> (Output_Get:double)
   n21 = Get bIsFalling()
   > Set Is Falling(Is Falling:double=float * float!(A="To Float (Boolean)"(InBool=@n21), B=@n20) -> (ReturnValue)) -> (Output_Get:double)
   > Set CrouchThighLRot(CrouchThighLRot:double=float * float(A=Get crouch alpha(), B=Get thigh_rotation_l() -> (thigh_rotation_l:double))) -> (Output_Get:double)
   > Set Is Sprinting(Is Sprinting:double="To Float (Boolean)"(InBool=Get bIsSprinting())) -> (Output_Get:double)
   Event OnBeginCrouching()
   > Set crouch alpha(crouch alpha:double=1) -> (Output_Get:double)
   Event OnEndCrouching()
   > Set crouch alpha(crouch alpha=0)
   Event BlueprintInitializeAnimation()
   > Set SkelMesh(SkelMesh=GetOwnerSkeletalMeshComponent())
   TryGetPawnOwner()
   Get rigid body enabled alpha()
  ]}
]
variables [
 Backward {type=number}
 crouch alpha {type=number}
 CrouchThighLRot {type=number}
 crouchThighRot {type=number}
 Forward {type=number}
 gravity_override {type=Vector}
 head l {type=number}
 head r {type=number}
 Is Falling {type=number}
 Is Sprinting {type=number}
 jumping alpha {type=number}
 Left {type=number}
 Right {type=number}
 Rigid Body Enabled {type=number}
 rigid body enabled alpha {type=number, default=1}
 rigid body enabled inv alpha {type=number}
 running_jumping_bool {type=boolean}
 SkelMesh {type=SkeletalMeshComponent}
 sprinting alpha {type=number}
 thigh rotation_r {type=number}
 thigh_l_rot {type=number}
 thigh_r_rot {type=number}
 thigh_rotation_l {type=number}
]
