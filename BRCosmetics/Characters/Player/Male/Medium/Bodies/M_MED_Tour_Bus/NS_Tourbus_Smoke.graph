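# Niagara smoke effect for the Tour Bus cosmetic body of the male medium player character.
# A mesh emitter makes short-lived, growing, randomly tinted particles, while a sprite emitter makes dark, rotating, SubUV-animated puffs, both shaped by fade curves and dynamic material parameters.
# Both sample skeletal mesh bones or sockets during update, using the particle partition to choose the bone and a mirrored offset.
# User.bisHandHidden zeroes spawn rate and particle size in both emitters.
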NiagaraSystem /BRCosmetics/Characters/Player/Male/Medium/Bodies/M_MED_Tour_Bus/NS_Tourbus_Smoke
dependencies {
 usedDataInterfaces [
  &NiagaraDataInterfaceVectorCurve
  &NiagaraDataInterfaceCurve
  &NiagaraDataInterfaceArrayFloat3
  &NiagaraDataInterfaceSpriteRendererInfo
  &NiagaraDataInterfaceSkeletalMesh
 ]
 usedDynamicInputs []
 usedModules [
  &self:SystemSpawnScript.SystemScriptSource.SystemScriptGraph.NiagaraNodeAssignment_0.SetVariables_#1
  &SystemState
 ]
}
emitters {
 Minimal {
  emitterUpdateScript [
   &EmitterState()
   &SpawnRate(
    SpawnRate=&MakeCustomFloatFromBool(Bool=$User.bisHandHidden, True Float=0, False Float=10)
    Use Spawn Probability=false
   )
  ]
  particleSpawnScript [
   &PartitionParticles()
   &InitializeParticle(
    Color Mode=::NewEnumerator0
    Lifetime Max=1.5
    Lifetime Mode=::NewEnumerator1
    Mesh Scale=&RandomRangeVector(Minimum=<0.03 0.003 1>, Maximum=<0.05 0.005 1>, Evaluation Type=::NewEnumerator1)
    Mesh Scale Mode=::NewEnumerator4
    UsePositionOffset=false
    Write Material Random=true
   )
   &DynamicMaterialParameters(
    Index 0 Param 2=&RandomRangeFloat(Minimum=20, Maximum=50, Evaluation Type=::NewEnumerator1)
    Module.Param0WriteEnabled=true
    Module.Param1WriteEnabled=true
    Module.Param2WriteEnabled=true
    Module.Param3WriteEnabled=true
   ) {module=DynamicMaterialParameters001}
   &Color(Color=&MakeLinearColorFromVectorAndFloat(
     "Vector (RGB)"=&RandomRangeVector(Minimum=<0 0.5 0>, Maximum=<1 1 1>, Evaluation Type=::NewEnumerator1)
     "Float (Alpha)"=20
    ))
   &AddVelocity(Velocity=<1 0 -1>)
  ]
  particleUpdateScript [
   &ParticleState()
   &ScaleMeshSize(Scale Factor=&VectorFromCurve(VectorCurve=DataInterface() {propertyValues {XCurve {keys [<at 0 0.1>,<at 1 10>]}, YCurve {keys [<at 0 1>,<at 1 8>]}, ZCurve {keys [<at 0 1>,<at 1 1>]}}}))
   &ScaleColor(
    Scale Alpha=&FloatFromCurve(FloatCurve=DataInterface() {propertyValues {Curve {keys [
         {arriveTangent=0.2972, interpMode=RCIM_Cubic, leaveTangent=0.2972, tangentMode=RCTM_User, time=0, value=-0.0092}
         {arriveTangent=2.8868, interpMode=RCIM_Cubic, leaveTangent=2.8868, tangentMode=RCTM_User, time=0.2459, value=1}
         {arriveTangent=-0.6384, interpMode=RCIM_Cubic, leaveTangent=-0.6384, time=0.3971, value=1.1998}
         {arriveTangent=-2.5872, interpMode=RCIM_Cubic, leaveTangent=-2.5872, time=0.6259, value=0.7574}
         {arriveTangent=-2.1791, interpMode=RCIM_Cubic, leaveTangent=-2.1791, time=0.7543, value=0.2756}
         {arriveTangent=-1.1217, interpMode=RCIM_Cubic, leaveTangent=-1.1217, time=0.9211, value=0.1142}
         <at 1 0 RCIM_Cubic>
        ]}}})
    Scale RGB=&VectorFromCurve(VectorCurve=DataInterface() {propertyValues {XCurve {keys [<at 0 1>,<at 0.3 1>]}, YCurve {keys [<at 0 3>,<at 0.3 0.1>]}, ZCurve {keys [<at 0 1>,<at 0.3 1>]}}})
    ScaleA=true
    ScaleRGB=true
   )
   &DynamicMaterialParameters(
    Index 0 Param 1=&FloatFromCurve(FloatCurve=DataInterface() {propertyValues {Curve {keys [<at 0 5>,<at 1 0>]}}})
    Index 0 Param 3=&FloatFromCurve(FloatCurve=DataInterface() {propertyValues {Curve {keys [<at 0 3>,<at 1 12>]}}})
    Index 0 Param 4=&FloatFromCurve(FloatCurve=DataInterface() {propertyValues {Curve {keys [
         {arriveTangent=-0.0169, interpMode=RCIM_Cubic, leaveTangent=-0.0169, tangentMode=RCTM_User, time=0.0123, value=0.25}
         {arriveTangent=-0.3638, interpMode=RCIM_Cubic, leaveTangent=-0.3638, tangentMode=RCTM_User, time=0.8212, value=0.2237}
         {arriveTangent=-2.1295, interpMode=RCIM_Cubic, leaveTangent=-2.1295, tangentMode=RCTM_User, time=0.9123, value=0.1}
         <at 1 0 RCIM_Cubic>
        ]}}})
    Module.Param0WriteEnabled=true
    Module.Param1WriteEnabled=false
    Module.Param2WriteEnabled=true
    Module.Param3WriteEnabled=true
   )
   &CameraOffset(Camera Offset Amount=2)
   &SkeletalMeshLocation(
    Bone / Socket Index=$Particles.PartitionParticles.Partition
    Bone Sampling Mode=::NewEnumerator4
    MaskSpawnGroup=false
    Sampled Position Offset=&SelectVectorFromArray(
     Direct Array Index=$Particles.PartitionParticles.Partition
     Vector Selection Array=DataInterface() {propertyValues {FloatData [<8 0 0>, <-8 0 0>]}}
    )
    Skeletal Mesh=$System.Skeletal Mesh
   ) {module=SkeletalMeshLocation001}
   &SolveForcesAndVelocity(Clamp Velocity=false, Limit Acceleration=false)
   &ScaleMeshSize(Scale Factor=&VectorFromFloat(Value=&MakeCustomFloatFromBool(Bool=$User.bisHandHidden, True Float=0, False Float=1))) {module=ScaleMeshSize001}
  ]
  rendererClasses [&NiagaraMeshRendererProperties]
  renderers [{rendererClass=&NiagaraMeshRendererProperties, rendererIndex=0}]
 }
 Minimal001 {
  emitterUpdateScript [
   &EmitterState()
   &SpawnRate(
    SpawnRate=&MakeCustomFloatFromBool(Bool=$User.bisHandHidden, True Float=0, False Float=3.2)
    Use Spawn Probability=false
   )
  ]
  particleSpawnScript [
   &PartitionParticles()
   /SetVariables(Particles.SpriteRotationRate=&RandomRangeFloat(Minimum=-0.15, Maximum=-0.3, Evaluation Type=::NewEnumerator1)) {isSetParameters=true}
   &InitializeParticle(
    Color Mode=::NewEnumerator0
    Lifetime Max=2.5
    Lifetime Min=1.6
    Lifetime Mode=::NewEnumerator1
    Sprite Rotation Angle=&RandomRangeFloat(Minimum=-1, Maximum=1, Evaluation Type=::NewEnumerator1)
    Sprite Rotation Mode=::NewEnumerator1
    Sprite Size=&Vector2DFromFloat(Value=&RandomRangeFloat(Minimum=18, Maximum=22, Evaluation Type=::NewEnumerator1))
    Sprite Size Mode=::NewEnumerator4
    UsePositionOffset=false
    Write Material Random=true
   )
   &Color(Color=&MakeLinearColorFromVectorAndFloat(
     "Vector (RGB)"=&RandomRangeVector(Minimum=<0.05 0.05 0.05>, Maximum=<0.15 0.15 0.15>, Evaluation Type=::NewEnumerator1)
     "Float (Alpha)"=&RandomRangeFloat(Minimum=0.4, Maximum=0.6, Evaluation Type=::NewEnumerator1)
    ))
   &CameraOffset(Camera Offset Amount=&RandomRangeFloat(Minimum=5, Maximum=8, Evaluation Type=::NewEnumerator1)) {module=CameraOffset001}
   &DynamicMaterialParameters(
    Index 0 Param 1=&RandomRangeFloat(Minimum=0, Maximum=1, Evaluation Type=::NewEnumerator1)
    Index 0 Param 2=&RandomRangeFloat(Minimum=1, Maximum=2, Evaluation Type=::NewEnumerator1)
    Module.Index0Float0Write=true
    Module.Index0Float1Write=true
    Module.Index0Float2Write=false
    Module.Index0Float3Write=false
   ) {module=DynamicMaterialParameters001}
   &AddVelocity(Velocity=&RandomRangeVector(Minimum=<6 -5 3>, Maximum=<16 5 5>, Evaluation Type=::NewEnumerator1))
  ]
  particleUpdateScript [
   &ParticleState()
   &ScaleSpriteSize(
    Scale Sprite Size Mode=::NewEnumerator0
    Uniform Scale Factor=&FloatFromCurve(FloatCurve=DataInterface() {propertyValues {Curve {keys [
         {arriveTangent=7.7033, interpMode=RCIM_Cubic, leaveTangent=7.7033, tangentMode=RCTM_User, time=0, value=0.2}
         {arriveTangent=0.8509, interpMode=RCIM_Cubic, leaveTangent=0.8509, time=0.0948, value=0.7009}
         <at 0.9401 1 RCIM_Cubic>
        ]}}})
   )
   &SpriteRotationRate(Rotation Rate=&AngleConversion(
     Angle:NiagaraFloat=$Particles.SpriteRotationRate
     Angle Input:ENiagara_AngleInput=::NewEnumerator1
     Angle Output:ENiagara_AngleInput=::NewEnumerator0
    ))
   &ScaleColor(
    Scale Alpha=&FloatFromCurve(FloatCurve=DataInterface() {propertyValues {Curve {keys [
         <at 0 0 RCIM_Cubic>
         {arriveTangent=1.9524, interpMode=RCIM_Cubic, leaveTangent=1.9524, tangentMode=RCTM_User, time=0.25, value=1}
         {arriveTangent=-1.3333, interpMode=RCIM_Cubic, leaveTangent=-1.3333, time=0.8, value=1.8}
         <at 1 0 RCIM_Cubic>
        ]}}})
    Scale RGB=&VectorFromCurve(VectorCurve=DataInterface() {propertyValues {XCurve {keys [<at 0 1>,<at 0.5 0.2>]}, YCurve {keys [<at 0 1>,<at 0.5 0.2>]}, ZCurve {keys [<at 0 1>,<at 0.5 0.2>]}}})
    ScaleA=true
    ScaleRGB=true
   )
   &DynamicMaterialParameters(
    Index 0 Param 1=&FloatFromCurve(FloatCurve=DataInterface() {propertyValues {Curve {keys [<at 0 5>,<at 1 0>]}}})
    Index 0 Param 3=&FloatFromCurve(FloatCurve=DataInterface() {propertyValues {Curve {keys [<at 0 3>,<at 1 12>]}}})
    Index 0 Param 4=&FloatFromCurve(FloatCurve=DataInterface() {propertyValues {Curve {keys [
         {arriveTangent=-0.0169, interpMode=RCIM_Cubic, leaveTangent=-0.0169, tangentMode=RCTM_User, time=0.0123, value=0.25}
         {arriveTangent=-0.3638, interpMode=RCIM_Cubic, leaveTangent=-0.3638, tangentMode=RCTM_User, time=0.8212, value=0.2237}
         {arriveTangent=-2.1295, interpMode=RCIM_Cubic, leaveTangent=-2.1295, tangentMode=RCTM_User, time=0.9123, value=0.1}
         <at 1 0 RCIM_Cubic>
        ]}}})
    Module.Param0WriteEnabled=true
    Module.Param1WriteEnabled=false
    Module.Param2WriteEnabled=true
    Module.Param3WriteEnabled=true
   )
   &CameraOffset(Camera Offset Amount=2) {enabled=false}
   &SkeletalMeshLocation(
    Bone / Socket Index=$Particles.PartitionParticles.Partition
    Bone Sampling Mode=::NewEnumerator4
    MaskSpawnGroup=false
    Sampled Position Offset=&SelectVectorFromArray(
     Direct Array Index=$Particles.PartitionParticles.Partition
     Vector Selection Array=DataInterface() {propertyValues {FloatData [<8 0 0>, <-8 0 0>]}}
    )
    Skeletal Mesh=$System.Skeletal Mesh
   )
   &SolveForcesAndVelocity(Clamp Velocity=false, Limit Acceleration=false)
   &SubUVAnimation(
    "Sprite Renderer "=DataInterface() {propertyValues {SpriteRenderer=&self:Minimal001_0.NiagaraSpriteRendererProperties_0}}
    UseEndFrame=false
    UseStartFrame=false
   )
   &DynamicMaterialParameters(
    Index 0 Param 4=0.5
    Module.Index0Float0Write=true
    Module.Index0Float1Write=true
    Module.Index0Float2Write=true
    Module.Index0Float3Write=true
   ) {module=DynamicMaterialParameters002}
   &ScaleSpriteSize(
    Scale Sprite Size Mode=::NewEnumerator0
    Uniform Scale Factor=&MakeCustomFloatFromBool(Bool=$User.bisHandHidden, True Float=0, False Float=1)
   ) {module=ScaleSpriteSize001}
  ]
  rendererClasses [&NiagaraSpriteRendererProperties]
  renderers [{rendererClass=&NiagaraSpriteRendererProperties}]
 }
}
summary {
 systemName=NS_Tourbus_Smoke
 emitters [
  {emitterName=Minimal, rendererClasses [&NiagaraMeshRendererProperties]}
  {emitterName=Minimal001, rendererClasses [&NiagaraSpriteRendererProperties]}
 ]
 userVariables [
  User.bisFootHidden {type=NiagaraBool}
  User.bisHandHidden {type=NiagaraBool}
  User.bisHeadHidden {type=NiagaraBool}
 ]
}
